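# Sound cue for the FmDrftFx pet cosmetic, played, going by its name, when the down-but-not-out (DBNO) state starts.
# A Random node picks among three equally weighted DBNO wave variations and feeds all four inputs of the root quality-level node, so every quality level sounds the same.
# The cue uses the close-range pet attenuation and the creature pets sound class, with a slight volume boost.
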SoundCue /BRCosmetics/Sounds/Pets/FmDrftFx/PT_DrftFxFm_DbnoStart_Cue
root=QualityLevel_0
nodes [
 {id=QualityLevel_0, children [Random_0,Random_0,Random_0,Random_0]}
 {id=Random_0, children [WavePlayer_3,WavePlayer_4,WavePlayer_6], properties {weights [1,1,1]}}
 {id=WavePlayer_3, properties {soundWaveAssetPtr=&/FmDrftFx_DBNO_01}}
 {id=WavePlayer_4, properties {soundWaveAssetPtr=&/FmDrftFx_DBNO_02}}
 {id=WavePlayer_6, properties {soundWaveAssetPtr=&/FmDrftFx_DBNO_03}}
]
properties {
 attenuationSettings=&Fort_Pets_Close_Athena_Att
 maxDistance=1800.0001
 soundClassObject=&Fort_Creatures_Pets
 volumeMultiplier=1.12
}
