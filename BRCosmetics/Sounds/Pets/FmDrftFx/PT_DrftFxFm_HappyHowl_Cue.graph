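# Sound cue for the happy howl of the FmDrftFx pet, classed under creature pet sounds with close-range pet attenuation.
# A Random node picks between two equally weighted Happy wave variants and feeds a Modulator.
# All four inputs of the root quality level node share that Modulator, so every quality setting plays the same chain.
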SoundCue /BRCosmetics/Sounds/Pets/FmDrftFx/PT_DrftFxFm_HappyHowl_Cue
root=QualityLevel_0
nodes [
 {id=Modulator_0, children [Random_0]}
 {id=QualityLevel_0, children [Modulator_0,Modulator_0,Modulator_0,Modulator_0]}
 {id=Random_0, children [WavePlayer_1,WavePlayer_2], properties {weights [1,1]}}
 {id=WavePlayer_1, properties {soundWaveAssetPtr=&/FmDrftFx_Happy_06}}
 {id=WavePlayer_2, properties {soundWaveAssetPtr=&/FmDrftFx_Happy_07}}
]
properties {
 attenuationSettings=&Fort_Pets_Close_Athena_Att
 maxDistance=1800.0001
 soundClassObject=&Fort_Creatures_Pets
 volumeMultiplier=1.12
}
