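# Sound cue for the FmDrftFx pet's skydiving sound, in the pets sound class with close-range pet attenuation.
# A short random delay feeds a quality-level switch whose four inputs share one random node, so every tier picks equally from three skydive waves.
# Overridden concurrency allows one instance per owner and blocks new plays while one is active.
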SoundCue /BRCosmetics/Sounds/Pets/FmDrftFx/PT_DrftFxFm_Skydiving_Cue
root=Delay_0
nodes [
 {id=Delay_0, children [QualityLevel_0], properties {delayMax=0.3, delayMin=0.2}}
 {id=QualityLevel_0, children [Random_0,Random_0,Random_0,Random_0]}
 {id=Random_0, children [WavePlayer_2,WavePlayer_3,WavePlayer_4], properties {weights [1,1,1]}}
 {id=WavePlayer_2, properties {soundWaveAssetPtr=&/FmDrftFx_Skydive_01}}
 {id=WavePlayer_3, properties {soundWaveAssetPtr=&/FmDrftFx_Skydive_02}}
 {id=WavePlayer_4, properties {soundWaveAssetPtr=&/FmDrftFx_Skydive_03}}
]
properties {
 attenuationSettings=&Fort_Pets_Close_Athena_Att
 bOverrideConcurrency=true
 maxDistance=1800.0001
 soundClassObject=&Fort_Creatures_Pets
 volumeMultiplier=0.896
 concurrencyOverrides {bLimitToOwner=true, maxCount=1, resolutionRule=PreventNew}
}
